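# Translucent, unlit, two-sided material for Niagara mesh particles, drawing a stage-show fire spray cone tinted by particle color.
# Panning noise, driven by dynamic parameters (NoiseSizeX scales both axes), multiplies particle alpha, an axis-aligned fresnel and an inverted-V fade.
# The fresnel axis bends from local X toward WhipDirection along a powered V gradient.
# The same WPOExp gradient scales the whip vector into an apparent vertex offset.
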Material /SparksMaps/Effects/StageShow/Fire/Materials/M_FireSprayCone
graph {
 properties {
  bUsedWithNiagaraMeshParticles=true
  blendMode=BLEND_Translucent
  pixelDepthOffsetMode=PDOM_Legacy
  previewMesh=/Game/Effects/Fort_Effects/Mesh/Jim/SM_Jim_Punch_Cone_01
  shadingModel=MSM_Unlit
  twoSided=true
  usedShadingModels=Unlit
 }
 nodes [
  n1 = ParticleColor()
  n2 = TextureCoordinate()
  n3 = DynamicParameter() {defaultValue=<rgba 0.1 1 1 1>, paramNames [NoiseSpeed, NoiseSizeX, NoiseSizeY, WhipDirection]}
  n4 = AppendVector(A=@n3.NoiseSizeX, B=@n3.NoiseSizeX)
  n5 = Time()
  n6 = Multiply(A=ScalarParameter() {defaultValue=1, parameterName=NoiseSpeed}, B=@n3.NoiseSpeed)
  n7 = TextureSampleParameter2D!(UVs=Panner(Coordinate=Multiply(A=@n2, B=@n4), Time=Multiply(A=@n5, B=@n6)) {speedX=0.1, speedY=-1}, Level, Apply View MipBias) {constMipValue=0, mipValueMode=TMVM_MipLevel, parameterName=NoiseTexture, texture=/Game/Effects/FogVolume/T_TileNoise}
  n8 = Constant()
  n9 = Transform(None=MaterialFunctionCall(
    X=DynamicParameter() {defaultValue=<rgba 0.1 1 1 1>, paramNames [NoiseSpeed, NoiseSizeX, NoiseSizeY, WhipDirection]}.WhipDirection
    Y=@n8
    Z=@n8
   ) {materialFunction=&MakeFloat3}.Result) {transformSourceType=TRANSFORMSOURCE_Instance}
  n10 = ComponentMask(None=TextureCoordinate()) {g=true}
  n11 = ScalarParameter() {defaultValue=5, parameterName=WPOExp}
  n12 = LinearInterpolate(
   A=Transform(None=Constant3Vector() {constant=<rgba 1 0 0 0>}) {transformSourceType=TRANSFORMSOURCE_Local}
   B=@n9
   Alpha=Saturate(None=Power(Base=@n10, Exp=@n11))
  )
  n13 = ScalarParameter() {defaultValue=5, parameterName=AxisExponent}
  n14 = ComponentMask(None=TextureCoordinate()) {g=true}
  Multiply(
   A=Multiply(
    A=Multiply(A=@n1.A, B=@n7.R)
    B=MaterialFunctionCall(Axis=@n12, Exponent=@n13) {materialFunction=&AxisAlignedFresnel}.Fresnel
   )
   B=OneMinus(None=@n14)
  )
  n15 = ScalarParameter() {defaultValue=5, parameterName=WPOExp}
  Multiply(A=Saturate(None=Power(Base=@n14, Exp=@n15)), B=@n9)
  Reroute(None=@n1.RGB)
  VectorParameter() {parameterName=Push}
 ]
}
