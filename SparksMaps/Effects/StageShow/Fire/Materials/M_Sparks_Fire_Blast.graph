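# Translucent unlit Niagara sprite and mesh material for stage-show fire blasts.
# Panned, distorted noise, with per-particle random flip and time offset, combines with procedural flame and center-body masks plus macro detail to build fire shape and color ramp.
# The ramp, scaled by dynamic Temperature, drives a curve-atlas gradient or a particle-color-only switch, while particle alpha erodes opacity.
# Dynamic Fire Speed drives panning; low quality skips distortion.
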Material /SparksMaps/Effects/StageShow/Fire/Materials/M_Sparks_Fire_Blast
graph {
 properties {
  bCastRayTracedShadows=false
  bUsedWithNiagaraMeshParticles=true
  bUsedWithNiagaraSprites=true
  blendMode=BLEND_Translucent
  pixelDepthOffsetMode=PDOM_Legacy
  shadingModel=MSM_Unlit
  usedShadingModels=Unlit
 }
 nodes [
  n1 = NamedRerouteUsage()
  n2 = Multiply(
   A=Add(A=NamedRerouteUsage(), B=Add(A=@n1) {constB=0.05})
   B=DynamicParameter() {defaultValue=<rgba 1 -2.5 1 1>, paramNames [Temperature, Fire Speed, Param3, Param4]}.Temperature
  ) {constB=2}
  n3 = ScalarParameter() {defaultValue=5, group=Fire Color Ramp Control, parameterName=Final Intensity Boost, sortPriority=2}
  n4 = Multiply(A=@n2, B=@n3)
  n5 = CurveAtlasRowParameter(CurveTime=@n2) {
   atlas=/SparksMaps/Effects/StageShow/Fire/Textures/CurveAtlas_StageFX_Fire
   curve=/SparksMaps/Effects/StageShow/Fire/Textures/Curve_StageFX_Fire_Natural
   group=Fire Color Ramp Control
   parameterName=ColorCurve
   sortPriority=3
  }
  n6 = Multiply(A=@n3, B=@n5)
  MaterialFunctionCall(Emissive=Multiply(A=ParticleColor().RGB, B=StaticSwitchParameter(True=@n4, False=@n6) {group=Fire Color Ramp Control, parameterName=Use Particle Color Only})) {materialFunction=&MF_FX_EmissiveAdjustments}
  n7 = ParticleColor()
  n8 = NamedRerouteUsage()
  n9 = Subtract(A=Saturate(None=Add(A=@n7.A, B=@n8)), B=OneMinus(None=@n7.A))
  Multiply(A=NamedRerouteUsage(), B=Saturate(None=Divide(A=@n9) {constB=1}))
  n10 = TextureCoordinate()
  n11 = ComponentMask(None=@n10) {r=true}
  n12 = Power(Base=Sine(None=Multiply(A=@n11) {constB=0.5}), Exp=Constant() {r=3.25}) {constExponent=10}
  n13 = Constant() {r=0.5}
  n14 = ComponentMask(None=OneMinus(None=@n10)) {g=true}
  n15 = Multiply(A=ComponentMask(None=@n10) {g=true}, B=Saturate(None=Multiply(A=@n14) {constB=3}))
  NamedRerouteDeclaration(None=Add(A=Multiply(A=@n12, B=@n13) {constB=0.3}, B=Multiply(A=@n12, B=@n15))) {name=Flame Shape Masks, nodeColor=<rgba 0 0 0 1>}
  n16 = Add(A=NamedRerouteUsage(), B=NamedRerouteUsage())
  n17 = NamedRerouteUsage()
  n18 = Subtract(
   A=Saturate(None=Multiply(A=@n16, B=@n17))
   B=ScalarParameter() {defaultValue=0.65, desc=0.5, group=Fire Shape Control, parameterName=Fire Shape Subtract, sortPriority=0}
  ) {constB=0.3}
  n19 = ScalarParameter() {defaultValue=0.5, desc=0.75, group=Fire Shape Control, parameterName=Fire Shape Feather, sortPriority=1}
  NamedRerouteDeclaration(None=Saturate(None=Divide(A=@n18, B=@n19) {constB=0.25})) {name=Fire Shape, nodeColor=<rgba 0 0 0 1>}
  n20 = Add(A=Multiply(A=NamedRerouteUsage()) {constB=0.2}, B=NamedRerouteUsage())
  n21 = NamedRerouteUsage()
  n22 = Subtract(
   A=Saturate(None=Multiply(A=@n20, B=@n21))
   B=ScalarParameter() {defaultValue=0.45, desc=0.3, group=Fire Color Ramp Control, parameterName=Color Ramp Subtract, sortPriority=0}
  ) {constB=0.3}
  n23 = ScalarParameter() {defaultValue=0.8, desc=1, group=Fire Color Ramp Control, parameterName=Color Ramp Feather, sortPriority=1}
  NamedRerouteDeclaration(None=Saturate(None=Divide(A=@n22, B=@n23) {constB=0.25})) {name=Fire Color Ramp, nodeColor=<rgba 0 0 0 1>}
  n24 = TextureCoordinate()
  n25 = ComponentMask(None=@n24) {r=true}
  n26 = OneMinus(None=@n25)
  n27 = AppendVector(A=If(A=ParticleRandom(), "A > B"=@n25, "A == B"=@n26, "A < B"=@n26) {constB=0.5}, B=OneMinus(None=ComponentMask(None=@n24) {g=true}))
  n28 = TextureCoordinate()
  n29 = ScalarParameter() {defaultValue=1, desc=1, group=Fire Texture, parameterName=Fire Tile X, sortPriority=0}
  n30 = ScalarParameter() {defaultValue=0.75, desc=0.75, group=Fire Texture, parameterName=Fire Tile Y, sortPriority=1}
  n31 = Add(
   A=Multiply(
    A=StaticSwitchParameter(True=@n27, False=@n28) {defaultValue=true, group=Fire Texture, parameterName=Flip Fire Texture}
    B=AppendVector(A=@n29, B=@n30)
   )
   B=ScalarParameter() {defaultValue=0.6, desc=0.1, group=Fire Texture, parameterName=Fire Tex Pos, sortPriority=2}
  )
  n32 = TextureSample(UVs=Panner(Coordinate=TextureCoordinate() {uTiling=4, vTiling=4}) {speedY=6}) {samplerType=SAMPLERTYPE_LinearGrayscale, texture=/Game/Effects/Fort_Effects/Textures/Tile/T_CellNoise_Tile_M}
  n33 = ScalarParameter() {defaultValue=0.01, group=Fire Texture, parameterName=Fire Distort Amount, sortPriority=3}
  n34 = Panner(
   Coordinate=QualitySwitch(Default=LinearInterpolate(A=@n31, B=@n32.R, Alpha=@n33), Low=@n31)
   Time=Add(A=Time(), B=ParticleRandom())
   Speed=AppendVector(A=Constant(), B=DynamicParameter() {defaultValue=<rgba 1 -2.5 1 1>, paramNames [Temperature, Fire Speed, Param3, Param4]}.Fire Speed)
  ) {speedY=0.75}
  NamedRerouteDeclaration(None=OneMinus(None=TextureSample(UVs=@n34) {samplerType=SAMPLERTYPE_LinearColor, texture=/SparksMaps/Effects/StageShow/Fire/Textures/T_Mojo_Fire_Noise}.G)) {name=FireTex, nodeColor=<rgba 0 0 0 1>}
  n35 = Constant() {r=0.65}
  n36 = Constant() {r=1.25}
  n37 = SphereMask(
   A=MaterialFunctionCall(Texture Scale=AppendVector(A=@n35, B=@n36)) {materialFunction=&ScaleUVsByCenter}.UVs
   B=Constant2Vector() {desc="0.5, 0.7", g=0.7, r=0.5}
  ) {attenuationRadius=0.3, hardnessPercent=0}
  NamedRerouteDeclaration(None=Multiply(A=Power(Base=@n37), B=Constant() {desc=Boost center body amount, r=2}) {constB=2, desc=Test - worked well for flipped texture fireball feel}) {name=Center Body, nodeColor=<rgba 0 0 0 1>}
  n38 = Multiply(
   A=TextureCoordinate()
   B=AppendVector(
    A=ScalarParameter() {defaultValue=3, group=Fire Macro Detail, parameterName=Fire Macro Tile X, sortPriority=0}
    B=ScalarParameter() {defaultValue=1.5, group=Fire Macro Detail, parameterName=Fire Macro Tile Y, sortPriority=1}
   )
  )
  n39 = Constant() {r=0.4}
  n40 = Time()
  n41 = ParticleRandom()
  n42 = Constant()
  n43 = ScalarParameter() {defaultValue=2.75, group=Fire Macro Detail, parameterName=Fire Macro Speed, sortPriority=2}
  n44 = TextureSample(UVs=Panner(Coordinate=Add(A=@n38, B=@n39), Time=Add(A=@n40, B=@n41), Speed=AppendVector(A=@n42, B=@n43)) {speedY=0.75}) {samplerType=SAMPLERTYPE_LinearColor, texture=&T_Mojo_Fire_Noise}
  n45 = Divide(A=Subtract(A=OneMinus(None=@n44.G), B=Constant() {r=0.5}), B=Constant() {r=0.5})
  NamedRerouteDeclaration(None=Multiply(A=Saturate(None=@n45), B=ScalarParameter() {defaultValue=0.5, group=Fire Macro Detail, parameterName=Fire Macro Amount, sortPriority=3}) {constB=0.2}) {name=Fire Macro Detail, nodeColor=<rgba 0 0 0 1>}
 ]
}
